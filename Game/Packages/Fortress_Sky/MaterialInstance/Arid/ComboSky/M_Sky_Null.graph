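# M_Sky_Null is an unlit material in the Fortress_Sky package's Arid ComboSky folder, apparently an empty placeholder for a sky material, going by its name.
# Its graph holds a single Constant node left at its default value, with no textures, parameters or visible connection to a material output.
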Material /Game/Packages/Fortress_Sky/MaterialInstance/Arid/ComboSky/M_Sky_Null
graph {properties {pixelDepthOffsetMode=PDOM_Legacy, shadingModel=MSM_Unlit, usedShadingModels=Unlit}, nodes [Constant()]}
